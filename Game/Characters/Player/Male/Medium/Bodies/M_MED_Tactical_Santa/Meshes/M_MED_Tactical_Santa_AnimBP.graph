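# Anim Blueprint for the Tactical Santa male body part, adding secondary motion on top of the pose copied from the attached parent mesh.
# AnimDynamics springs the front and back straps, then two RBF Pose Drivers drive bicep bones from forearm pose before a Control Rig finishes the pose.
# When customization loads, a dynamic material parameter sets "Hide Element 01" on if the character item has the TacticalSantaReindeer meta tag, off otherwise.
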AnimBlueprint /Game/Characters/Player/Male/Medium/Bodies/M_MED_Tactical_Santa/Meshes/M_MED_Tactical_Santa_AnimBP
target_skeleton=&/M_MED_Tactical_Santa_Skeleton
graphs [
 AnimGraph {nodes [
   n1 = Anim Dynamics - Bone: dyn_front_strap(ComponentPose=Local To Component(LocalPose=Copy Pose From Mesh() {bUseAttachedParent=true}), Alpha=1, ExternalForce="0,0,0") {
    angularDampingOverride=1
    angularSpringConstant=20000
    bAngularSpring=true
    bOverrideAngularDamping=true
    bUseGravityOverride=true
    gravityScale=0.35
    simulationSpace=RootRelative
    alphaBoolBlend {blendOption=Linear}
    boundBone {boneName=dyn_front_strap}
    componentLinearAccScale=<1 1 1>
    componentLinearVelScale=<1 1 1>
    gravityOverride {z=-980}
    physicsBodyDefinitions [{
      boundBone {boneName=dyn_front_strap}
      boxExtents=<40 40 30>
      constraintSetup {angularLimitsMax=<20 20 0>, angularLimitsMin=<-20 -20 -20>, angularTarget=<1 0 0>}
      localJointOffset=<4 0 0>
     }]
   }
   n2 = Anim Dynamics - Bone: dyn_back_strap(ComponentPose=@n1, Alpha=1, ExternalForce="0,0,0") {
    angularDampingOverride=1
    angularSpringConstant=20000
    bAngularSpring=true
    bOverrideAngularDamping=true
    bUseGravityOverride=true
    gravityScale=0.35
    simulationSpace=RootRelative
    alphaBoolBlend {blendOption=Linear}
    boundBone {boneName=dyn_back_strap}
    componentLinearAccScale=<1 1 1>
    componentLinearVelScale=<1 1 1>
    gravityOverride {z=-980}
    physicsBodyDefinitions [{
      boundBone {boneName=dyn_back_strap}
      boxExtents=<40 40 30>
      constraintSetup {angularLimitsMax=<20 20 0>, angularLimitsMin=<-20 -20 -20>, angularTarget=<1 0 0>}
      localJointOffset=<4 0 0>
     }]
   }
   n3 = Pose Driver - Source: lowerarm_l!(SourcePose:Pose Link Structure=Component To Local(ComponentPose=@n2)) -> (Pose:Pose Link Structure) {
    is=title AnimGraphNode_PoseDriver
    poseAsset=&/M_MED_Tactical_Santa_Biceps_PoseAsset
    onlyDriveBones [{boneName=dyn_bicep_l}]
    poseTargets [
     {
      bApplyCustomCurve=false
      bIsHidden=false
      distanceMethod=DefaultMethod
      drivenName=Pose_1
      functionType=DefaultFunction
      targetScale=1
      boneTransforms [{targetRotation=<rot -4.1538 115.966 -2.0203>, targetTranslation=<26.6514 0 0>}]
      customCurve {defaultValue=3.4028234663852886e+38, postInfinityExtrap=RCCE_Constant, preInfinityExtrap=RCCE_Constant, keys []}
      targetRotation=<rot 0 0 0>
     }
     {
      bApplyCustomCurve=false
      bIsHidden=false
      distanceMethod=DefaultMethod
      drivenName=Pose_2
      functionType=DefaultFunction
      targetScale=0.2401
      boneTransforms [{targetRotation=<rot -0.5954 7.3908 4.5799>, targetTranslation=<26.6514 0 0>}]
      customCurve {defaultValue=3.4028234663852886e+38, postInfinityExtrap=RCCE_Constant, preInfinityExtrap=RCCE_Constant, keys []}
      targetRotation=<rot 0 0 0>
     }
    ]
    rBFParams {function=Exponential, radius=86.2555}
    sourceBones [{boneName=lowerarm_l}]
   }
   n4 = Pose Driver - Source: lowerarm_r!(SourcePose:Pose Link Structure=@n3) -> (Pose:Pose Link Structure) {
    is=title AnimGraphNode_PoseDriver
    poseAsset=&M_MED_Tactical_Santa_Biceps_PoseAsset
    onlyDriveBones [{boneName=dyn_bicep_r}]
    poseTargets [
     {
      bApplyCustomCurve=false
      bIsHidden=false
      distanceMethod=DefaultMethod
      drivenName=Pose_1
      functionType=DefaultFunction
      targetScale=1
      boneTransforms [{targetRotation=<rot -4.1538 115.966 -2.0203>, targetTranslation=<-26.6515 0 0>}]
      customCurve {defaultValue=3.4028234663852886e+38, postInfinityExtrap=RCCE_Constant, preInfinityExtrap=RCCE_Constant, keys []}
      targetRotation=<rot 0 0 0>
     }
     {
      bApplyCustomCurve=false
      bIsHidden=false
      distanceMethod=DefaultMethod
      drivenName=Pose_2
      functionType=DefaultFunction
      targetScale=0.2401
      boneTransforms [{targetRotation=<rot -0.5954 7.3908 4.5799>, targetTranslation=<-26.6515 0 0>}]
      customCurve {defaultValue=3.4028234663852886e+38, postInfinityExtrap=RCCE_Constant, preInfinityExtrap=RCCE_Constant, keys []}
      targetRotation=<rot 0 0 0>
     }
    ]
    rBFParams {function=Exponential, radius=86}
    sourceBones [{boneName=lowerarm_r}]
   }
   n5 = Control Rig(Source=@n4) {controlRigAssetReference {blueprintRigClass=&/M_MED_Tactical_Santa_CR.M_MED_Tactical_Santa_CR_C}}
   Output Pose(Result=@n5) {layerGroup=None}
  ]}
 EventGraph {nodes [
   Event OnCharacterCustomizationLoaded()
   > n1 = Get Dynamic Material Instance(SkeletalMeshComponent=GetPartSkeletalMeshComponent())
   > n2 = Branch(Condition=CharacterItemDefinitionHasAnyMetaTag(Tags=Make Array("[0]"=(TagName="Cosmetics.Metatags.TacticalSantaReindeer"))))
   > SetScalarParameterValue(self=@n1, ParameterName=Hide Element 01, Value=1)
   SetScalarParameterValue(execute=@n2.else, self=@n1, ParameterName=Hide Element 01, Value=0)
  ]}
 NewFunction {nodes [NewFunction() {is=title K2Node_FunctionEntry}]}
]
variables []
