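# Physics asset for secondary motion of the dangling accessories on the medium male Tactical Santa player body.
# Four identically sized box bodies for the candy canes each constrain to a small holster capsule, with tight swing and much wider twist.
# A back strap capsule constrains to spine_03, nearly rigid except along one swing axis.
# No constraint links the holster to the spine, and only these accessory bones and spine_03 have bodies.
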PhysicsAsset /Game/Characters/Player/Male/Medium/Bodies/M_MED_Tactical_Santa/Meshes/M_MED_Tactical_Santa_Dyn
bodies {
 dyn_back_strap [{shapeType=Capsule, length=4.8477, radius=1.6739, center=<0.05 0 0>, rotation=<rot 90 0 0>}]
 dyn_candycane_01 [{shapeType=Box, extentX=19.7208, extentY=1.5454, extentZ=7.5, center=<-0.0146 -0.0001 1.6554>}]
 dyn_candycane_02 [{shapeType=Box, extentX=19.7208, extentY=1.5454, extentZ=7.5, center=<-0.0146 -0.0001 1.6554>}]
 dyn_candycane_03 [{shapeType=Box, extentX=19.7208, extentY=1.5454, extentZ=7.5, center=<-0.0146 -0.0001 1.6554>}]
 dyn_candycane_04 [{shapeType=Box, extentX=19.7208, extentY=1.5454, extentZ=7.5, center=<-0.0146 -0.0001 1.6554>}]
 dyn_candycane_holster [{shapeType=Capsule, length=1, radius=1}]
 spine_03 [{shapeType=Capsule, length=2, radius=2}]
}
constraints [
 {bone1Name=dyn_candycane_01, bone2Name=dyn_candycane_holster, swing1LimitDegrees=2, swing2LimitDegrees=2, twistLimitDegrees=50}
 {bone1Name=dyn_candycane_02, bone2Name=dyn_candycane_holster, swing1LimitDegrees=2, swing2LimitDegrees=2, twistLimitDegrees=50}
 {bone1Name=dyn_candycane_03, bone2Name=dyn_candycane_holster, swing1LimitDegrees=2, swing2LimitDegrees=2, twistLimitDegrees=50}
 {bone1Name=dyn_candycane_04, bone2Name=dyn_candycane_holster, swing1LimitDegrees=2, swing2LimitDegrees=2, twistLimitDegrees=50}
 {bone1Name=dyn_back_strap, bone2Name=spine_03, swing1LimitDegrees=1, swing2LimitDegrees=10, twistLimitDegrees=1}
]
